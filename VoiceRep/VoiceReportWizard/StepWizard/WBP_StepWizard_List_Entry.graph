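# List entry widget for the voice report step wizard, showing a step header above that step's content widget.
# When given a BP_UI_StepWizard item, it removes the previous content, then creates, adds and activates the activatable widget class from the item's step data.
# SetStepTypeAndCount switches on step type to show a step counter, a green checkmark "Thanks" header, or a header without icon or counter.
# On construct, the step icon size is chosen by IsMobileGame.
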WidgetBlueprint /VoiceRep/VoiceReportWizard/StepWizard/WBP_StepWizard_List_Entry
graphs [
 EventGraph {nodes [
   Event Construct()
   > SetDesiredSizeOverride!(
    execute
    self:Image Object Reference=Get Step_Image() -> (Step_Image:Image Object Reference)
    DesiredSize:Vector 2D Structure=Select(Option 0:Vector 2D Structure=(X=16,Y=16), Option 1:Vector 2D Structure=(X=24,Y=24), Index=IsMobileGame()) -> (ReturnValue:Vector 2D Structure)
   ) -> (then)
   n1 = Event OnListItemObjectSet()
   > n2 = Cast To BP_UI_StepWizard!(execute, Object=@n1.ListItemObject) -> (then, CastFailed, AsBP UI Step Wizard:BP UI Step Wizard Object Reference) {is=title K2Node_DynamicCast}
   > n3 = Set ItemData(ItemData:BP UI Step Wizard Object Reference=@n2) -> (Output_Get:BP UI Step Wizard Object Reference)
   > RemoveFromParent(self=Get ContentWidget() -> (ContentWidget:User Widget Object Reference))
   > n4 = Create Widget!(
    execute:Exec
    Class:User Widget Class Reference=Break S UI Step Wizard Entry!(S_UI_StepWizard_Entry:"S UI Step Wizard Entry Structure (by ref)"=Get StepContentData!(self:BP UI Step Wizard Object Reference=@n3) -> (StepContentData:S UI Step Wizard Entry Structure)) -> (
     StepWidgetContent_92_#1:Common Activatable Widget Class Reference
     StepType_76_#2:E_UI_StepWizard_Type Enum
     CurrentStepNameText_82_#3:Text
     ButtonBack_90_#4:bool=~
     ButtonClose_91_#5:bool=~
    ) {is=title K2Node_BreakStruct}
    OwningPlayer:Player Controller Object Reference=GetOwningPlayer()
   ) -> (then:Exec, ReturnValue:Common Activatable Widget Object Reference)
   > n5 = Set ContentWidget(ContentWidget:User Widget Object Reference=@n4) -> (Output_Get:User Widget Object Reference)
   > SetStepTypeAndCount()
   > AddChildToVerticalBox(self=Get VerticalBox() -> (VerticalBox:Vertical Box Object Reference), Content=@n5)
   > n6 = Cast To CommonActivatableWidget!(execute, Object=@n5) -> (then, CastFailed, AsCommon Activatable Widget:Common Activatable Widget Object Reference) {is=title K2Node_DynamicCast}
   > ActivateWidget(self=@n6)
  ]}
 SetStepTypeAndCount {nodes [
   SetStepTypeAndCount() {is=title K2Node_FunctionEntry}
   n1 = Get ItemData() -> (ItemData:BP UI Step Wizard Object Reference)
   > n2 = Switch on E_UI_StepWizard_Type!(
    execute
    Selection:E_UI_StepWizard_Type Enum=Break S UI Step Wizard Entry(S_UI_StepWizard_Entry=Get StepContentData(self=@n1)) -> (ButtonBack_90_#4=~, ButtonClose_91_#5=~).StepType_76_#2
   ) -> (NewEnumerator0:Exec, NewEnumerator1:Exec, NewEnumerator2:Exec) {is=title K2Node_SwitchEnum}
   > SetVisibility(self=Get CountBox() -> (CountBox:Horizontal Box Object Reference), InVisibility=Visible)
   > "SetText (Text)"(self=Get Step_Name() -> (Step_Name:Common Text Object Reference), InText="LOCTEXT(\"Step\")")
   > SetBrushFromMaterial(self=Get Step_Image(), Material=/VoiceRep/Materials/MI_UI_StepDot_Outline)
   > Set ContentColorAndOpacity!(
    execute
    ContentColorAndOpacity:Linear Color Structure=(R=1,G=1,B=1,A=1)
    self:Border Object Reference=Get Border() -> (Border:Border Object Reference)
   ) -> (then, Output_Get:Linear Color Structure)
   n3 = Get CurrentStep!(self:BP UI Step Wizard Object Reference=Get ItemData()) -> (CurrentStep:int=0)
   > "SetText (Text)"(self=Get CurrentStep!() -> (CurrentStep:Common Text Object Reference), InText="To Text (Integer)"(Value=int + int(A=@n3)))
   n4 = Get ItemData()
   > "SetText (Text)"(
    self=Get TotalSteps() -> (TotalSteps:Common Text Object Reference)
    InText="To Text (Integer)"(Value=Get TotalSteps!(self:BP UI Step Wizard Object Reference=@n4) -> (TotalSteps:int=0))
   )
   SetVisibility(execute=@n2.NewEnumerator1, self=Get CountBox(), InVisibility=Collapsed)
   > "SetText (Text)"(self=Get Step_Name(), InText="LOCTEXT(\"Thanks\")")
   > SetBrushFromTexture!(
    execute
    self:Image Object Reference=Get Step_Image()
    Texture:Texture 2D Object Reference=/Game/Athena/UI/InGame/Creative/PropertyEditor/Icon_Picker/Textures/64x64/T_UI_IconLibrary_checkmark_64
    bMatchSize:bool=false
   ) -> (then)
   > Set ContentColorAndOpacity(ContentColorAndOpacity=(R=0.0369,G=1,B=0.2232,A=1), self=Get Border())
   SetVisibility(execute=@n2.NewEnumerator2, self=Get CountBox(), InVisibility=Collapsed)
   > "SetText (Text)"(self=Get Step_Name())
   > SetBrushFromMaterial(self=Get Step_Image(), Material=&MI_UI_StepDot_Outline)
   > SetVisibility(self=Get Step_Image(), InVisibility=Collapsed)
   > Set ContentColorAndOpacity(ContentColorAndOpacity=(R=1,G=1,B=1,A=1), self=Get Border())
  ]}
]
variables [
 ContentWidget {type=UserWidget, default=null}
 ItemData {type=BP_UI_StepWizard_C, default=null}
]
widgets {
 info {inheritedWidgetCount=0, parentClass=&UserWidget, rootWidgetClass=&SizeBox, widgetCount=11}
 widgets [
  SizeBox_179 {widgetClassPath=&SizeBox, widget=SizeBox_179, properties {bOverride_HeightOverride=true, heightOverride=652}}
  VerticalBox {widgetClassPath=&VerticalBox, bIsVariable=true, parent=SizeBox_179, widget=VerticalBox, slot {class=SizeBoxSlot}}
  Border {
   widgetClassPath=&Border
   bIsVariable=true
   parent=VerticalBox
   widget=Border
   properties {
    horizontalAlignment=HAlign_Left
    verticalAlignment=VAlign_Top
    background {
     bIsDynamicallyLoaded=false
     drawAs=Image
     imageType=NoImage
     mirroring=NoMirror
     resourceName=None
     resourceObject=/VoiceRep/Materials/MI_UI_LinearGradient
     tiling=NoTile
     imageSize=<32 32>
     margin=<ltrb 0 0 0 0>
     outlineSettings {
      bUseBrushTransparency=false
      roundingType=HalfHeightRadius
      width=0
      color {colorUseRule=UseColor_Specified, specifiedColor=<rgba 0 0 0 0>}
      cornerRadii=<0 0 0 0>
     }
     tintColor {colorUseRule=UseColor_Specified, specifiedColor=<rgba 1 1 1 1>}
     uVRegion {bIsValid=false, max=<0 0>, min=<0 0>}
    }
    padding=<ltrb 0 0 0 0>
   }
   slot {class=VerticalBoxSlot, properties {verticalAlignment=VAlign_Top}}
  }
  CounterStepBox {
   widgetClassPath=&HorizontalBox
   parent=Border
   widget=CounterStepBox
   slot {class=BorderSlot, properties {horizontalAlignment=HAlign_Left, verticalAlignment=VAlign_Top, padding=<ltrb 0 0 0 0>}}
  }
  Spacer_80 {widgetClassPath=&Spacer, parent=CounterStepBox, widget=Spacer_80, properties {size=<0 48>}, slot {class=HorizontalBoxSlot}}
  Step_Image {
   widgetClassPath=&Image
   bIsVariable=true
   parent=CounterStepBox
   widget=Step_Image
   properties {brush {
     bIsDynamicallyLoaded=false
     drawAs=Image
     imageType=NoImage
     mirroring=NoMirror
     resourceName=None
     resourceObject=&MI_UI_StepDot_Outline
     tiling=NoTile
     imageSize=<16 16>
     margin=<ltrb 0 0 0 0>
     outlineSettings {
      bUseBrushTransparency=false
      roundingType=HalfHeightRadius
      width=0
      color {colorUseRule=UseColor_Specified, specifiedColor=<rgba 0 0 0 0>}
      cornerRadii=<0 0 0 0>
     }
     tintColor {colorUseRule=UseColor_Specified, specifiedColor=<rgba 1 1 1 1>}
     uVRegion {bIsValid=false, max=<0 0>, min=<0 0>}
    }}
   slot {class=HorizontalBoxSlot, properties {horizontalAlignment=HAlign_Left, verticalAlignment=VAlign_Center, padding=<ltrb 0 0 8 0>}}
  }
  Step_Name {
   widgetClassPath=&CommonTextBlock
   bIsVariable=true
   parent=CounterStepBox
   widget=Step_Name
   properties {
    bIsScrollingEnabled=false
    mobileFontSizeMultiplier=1.5
    text="{STEP COUNTER NAME}"
    textTransformPolicy=ToUpper
    visibility=Visible
    wrapTextAt=600
    colorAndOpacity {colorUseRule=UseColor_Specified, specifiedColor=<rgba 0.0331 0.7011 1 1>}
    font {
     bForceMonospaced=false
     bMaterialIsStencil=false
     fontMaterial=null
     fontObject=&BurbankSmall
     letterSpacing=0
     monospacedWidth=1
     size=16
     skewAmount=0
     typefaceFontName=Bold
     outlineSettings {
      bApplyOutlineToDropShadows=false
      bMiteredCorners=false
      bSeparateFillAlpha=false
      outlineMaterial=null
      outlineSize=0
      outlineColor=<rgba 0 0 0 1>
     }
    }
    shadowOffset=<0 0>
   }
   slot {class=HorizontalBoxSlot, properties {horizontalAlignment=HAlign_Left, verticalAlignment=VAlign_Center, padding=<ltrb 8 0 16 0>}}
  }
  CountBox {widgetClassPath=&HorizontalBox, bIsVariable=true, parent=CounterStepBox, widget=CountBox, slot {class=HorizontalBoxSlot}}
  CurrentStep {
   widgetClassPath=&CommonTextBlock
   bIsVariable=true
   parent=CountBox
   widget=CurrentStep
   properties {
    bIsScrollingEnabled=false
    mobileFontSizeMultiplier=1.5
    text=n
    visibility=Visible
    wrapTextAt=600
    colorAndOpacity {colorUseRule=UseColor_Specified, specifiedColor=<rgba 0.0331 0.7011 1 1>}
    font {
     bForceMonospaced=false
     bMaterialIsStencil=false
     fontMaterial=null
     fontObject=&BurbankSmall
     letterSpacing=0
     monospacedWidth=1
     size=16
     skewAmount=0
     typefaceFontName=Bold
     outlineSettings {
      bApplyOutlineToDropShadows=false
      bMiteredCorners=false
      bSeparateFillAlpha=false
      outlineMaterial=null
      outlineSize=0
      outlineColor=<rgba 0 0 0 1>
     }
    }
    shadowOffset=<0 0>
   }
   slot {class=HorizontalBoxSlot, properties {horizontalAlignment=HAlign_Left, verticalAlignment=VAlign_Center}}
  }
  Separator {
   widgetClassPath=&CommonTextBlock
   parent=CountBox
   widget=Separator
   properties {
    bIsScrollingEnabled=false
    mobileFontSizeMultiplier=1.5
    text=/
    visibility=Visible
    wrapTextAt=600
    colorAndOpacity {colorUseRule=UseColor_Specified, specifiedColor=<rgba 0.0331 0.7011 1 1>}
    font {
     bForceMonospaced=false
     bMaterialIsStencil=false
     fontMaterial=null
     fontObject=&BurbankSmall
     letterSpacing=0
     monospacedWidth=1
     size=16
     skewAmount=0
     typefaceFontName=Bold
     outlineSettings {
      bApplyOutlineToDropShadows=false
      bMiteredCorners=false
      bSeparateFillAlpha=false
      outlineMaterial=null
      outlineSize=0
      outlineColor=<rgba 0 0 0 1>
     }
    }
    shadowOffset=<0 0>
   }
   slot {class=HorizontalBoxSlot, properties {horizontalAlignment=HAlign_Left, verticalAlignment=VAlign_Center}}
  }
  TotalSteps {
   widgetClassPath=&CommonTextBlock
   bIsVariable=true
   parent=CountBox
   widget=TotalSteps
   properties {
    bIsScrollingEnabled=false
    mobileFontSizeMultiplier=1.5
    text=n
    visibility=Visible
    wrapTextAt=600
    colorAndOpacity {colorUseRule=UseColor_Specified, specifiedColor=<rgba 0.0331 0.7011 1 1>}
    font {
     bForceMonospaced=false
     bMaterialIsStencil=false
     fontMaterial=null
     fontObject=&BurbankSmall
     letterSpacing=0
     monospacedWidth=1
     size=16
     skewAmount=0
     typefaceFontName=Bold
     outlineSettings {
      bApplyOutlineToDropShadows=false
      bMiteredCorners=false
      bSeparateFillAlpha=false
      outlineMaterial=null
      outlineSize=0
      outlineColor=<rgba 0 0 0 1>
     }
    }
    shadowOffset=<0 0>
   }
   slot {class=HorizontalBoxSlot, properties {horizontalAlignment=HAlign_Left, verticalAlignment=VAlign_Center}}
  }
 ]
}
